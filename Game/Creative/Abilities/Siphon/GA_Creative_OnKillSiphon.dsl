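; Gameplay ability for kill siphon: an event-triggered reward that restores health, shields and resources to a player who is not downed.
; AthenaGameData curve rows, or a debug bypass, enable the health and material halves separately.
; Amounts come from game state siphon overrides in Creative/Playground, otherwise from the data table.
; Healing uses SetByCaller effects, with overflow optionally converted to shields, while wood, stone, metal and gold go straight into inventory.
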(event Ability|EventActivateAbilityFromEvent
  (params (EventData :type Gameplay%20Event%20Data%20Structure%20%28by%20ref%29))
  (|CallParentFunction EventData)
  (bind _shouldcontinue (CallFunction|SetupRestoreModes))
  (if _shouldcontinue
    (if (Variables|MapState|GetbAttemptHealthRestore)
      (CallFunction|AttemptHeal))
    (if (Variables|MapState|GetbAttemptMatsRestore)
      (CallFunction|AttemptMats))
    (Ability|EndAbility)
    (else
      (Ability|EndAbility))))

(event Ability|EventOnEndAbility
  (params (bWasCancelled :type Boolean)))

(fn AddShields
  (params (bShouldConvertExcessHealthToShields :type Boolean) (ShieldToAdd :type Float%20%28double-precision%29))
  (if (and bShouldConvertExcessHealthToShields (> (Variables|GetShieldToAdd) 0.0))
    (bind _returnvalue (Ability|GameplayEffect|AssignTagSetbyCallerMagnitude (Ability|MakeOutgoingGameplayEffectSpec (Variables|AddShields|GetShieldGE)) "(TagName=\"SetByCaller.Shield\")" (Variables|GetShieldToAdd)))
    (Ability|ApplyGameplayEffectSpecToOwner _returnvalue)
    (return true (Variables|GetShieldToAdd))
    (else
      (Development|PrintString (select bShouldConvertExcessHealthToShields "Game Mode Does not Support converting Excess Health To Shields" "Game Mode Supports Excess Health To Shields, but value to apply is Zero."))
      (return false 0.0))))

(fn AttemptHeal
  (bind (_success _initialhealamount _actualappliedheal _excesshealing) (CallFunction|RestoreHealth))
  (if (>= _excesshealing 0.0)
    (bind _baddedshields (CallFunction|AddShields (Variables|AddShields|GetbShouldConvertExcessHealthToShields) _excesshealing))
    (else
      _))
  (if (or _baddedshields _success)
    (Ability|ExecuteGameplayCueOnOwner (Variables|Default|GetConsumedCue))
    (Development|PrintString "Player was Healed / Shield Restored On Elimination")))

(fn AttemptMats
  (bind (_bsuccess _resourcetogive _resourcetype) (CallFunction|CalculateResources "Wood"))
  (CallFunction|GiveResourcestoPlayer _bsuccess _resourcetogive _resourcetype)
  (bind (_bsuccess_2 _resourcetogive_2 _resourcetype_2) (CallFunction|CalculateResources "Stone"))
  (CallFunction|GiveResourcestoPlayer _bsuccess_2 _resourcetogive_2 _resourcetype_2)
  (bind (_bsuccess_3 _resourcetogive_3 _resourcetype_3) (CallFunction|CalculateResources "Metal"))
  (CallFunction|GiveResourcestoPlayer _bsuccess_3 _resourcetogive_3 _resourcetype_3)
  (bind (_bsuccess_4 _resourcetogive_4 _resourcetype_4) (CallFunction|CalculateResources "GoldCurrency"))
  (CallFunction|GiveResourcestoPlayer _bsuccess_4 _resourcetogive_4 _resourcetype_4))

(fn CalculateResources
  (params (ResourceGrantType :type EFortResourceType%20Enum :default "Wood"))
  (if (Variables|MapState|GetIsCreativeOrPlayground)
    (switch Utilities|FlowControl|Switch|SwitchonEFortResourceType (Variables|GetResourceGrantType)
      (:Wood
        (LocalVariables|SetiLocalResourceToGive (Math|Float|Truncate (Class|FortGameStateAthena|GetOverridenSiphonWoodValue (Variables|MapState|GetGameState) (Variables|Default|GetAbilityOwner)))))
      (:Stone
        (LocalVariables|SetiLocalResourceToGive (Math|Float|Truncate (Class|FortGameStateAthena|GetOverridenSiphonStoneValue (Variables|MapState|GetGameState) (Variables|Default|GetAbilityOwner)))))
      (:Metal
        (LocalVariables|SetiLocalResourceToGive (Math|Float|Truncate (Class|FortGameStateAthena|GetOverridenSiphonMetalValue (Variables|MapState|GetGameState) (Variables|Default|GetAbilityOwner)))))
      (:Permanite
        (Development|PrintString "***WARNING*** - Invalid Resource type to grant!")
        (return false 0 "None"))
      (:GoldCurrency
        (LocalVariables|SetiLocalResourceToGive (Math|Float|Truncate (Class|FortGameStateAthena|GetOverridenSiphonGoldValue (Variables|MapState|GetGameState) (Variables|Default|GetAbilityOwner)))))
      (:None
        _))
    (return (> (LocalVariables|GetiLocalResourceToGive) 0) (LocalVariables|GetiLocalResourceToGive) (Variables|GetResourceGrantType))
    (else
      (switch Utilities|FlowControl|Switch|SwitchonEFortResourceType (Variables|GetResourceGrantType)
        (:Wood
          (LocalVariables|SetLocalDataTableRowName (Variables|CalculateResources|GetNameGrantWood)))
        (:Stone
          (LocalVariables|SetLocalDataTableRowName (Variables|CalculateResources|GetNameGrantStone)))
        (:Metal
          (LocalVariables|SetLocalDataTableRowName (Variables|CalculateResources|GetNameGrantMetal)))
        (:Permanite
          _)
        (:GoldCurrency
          _)
        (:None
          _))
      (bind _outxy (DataTable|EvaluateCurveTableRow "/Game/Athena/Balance/DataTables/AthenaGameData.AthenaGameData" (LocalVariables|GetLocalDataTableRowName))
        (:RowFound
          (bind _output_get (LocalVariables|SetiLocalResourceToGive (Math|Float|Truncate _outxy)))
          (return (> _output_get 0) (LocalVariables|GetiLocalResourceToGive) (Variables|GetResourceGrantType)))
        (:RowNotFound
          (Development|PrintString "***WARNING*** - Attempted to Add Resource from Table but Row is Undefined!!")
          (return false 0 "None"))))))

(fn DetermineHealthDelta
  (params (HealingAmount :type Float%20%28double-precision%29))
  (return (FortPawn|GetMaxHealth (Variables|Default|GetAbilityOwner)) (FortPawn|GetHealth (Variables|Default|GetAbilityOwner)) (- (FortPawn|GetMaxHealth (Variables|Default|GetAbilityOwner)) (FortPawn|GetHealth (Variables|Default|GetAbilityOwner))) (select (>= (- HealingAmount (- (FortPawn|GetMaxHealth (Variables|Default|GetAbilityOwner)) (FortPawn|GetHealth (Variables|Default|GetAbilityOwner)))) 0.0) (- HealingAmount (- (FortPawn|GetMaxHealth (Variables|Default|GetAbilityOwner)) (FortPawn|GetHealth (Variables|Default|GetAbilityOwner)))) 0.0)))

(fn GiveResourcesToPlayer
  (params (bShouldGrant :type Boolean) (GiveAmount :type Integer) (GiveType :type EFortResourceType%20Enum :default "Wood"))
  (bind _abilityowner (Variables|Default|GetAbilityOwner))
  (bind _returnvalue (Fortnite|GetFortPlayerControllerfromActor _abilityowner))
  (if bShouldGrant
    (switch Utilities|FlowControl|Switch|SwitchonEFortResourceType GiveType
      (:Wood
        (LocalVariables|SetLocalItemToGive "/Game/Items/ResourcePickups/WoodItemData.WoodItemData"))
      (:Stone
        (LocalVariables|SetLocalItemToGive "/Game/Items/ResourcePickups/StoneItemData.StoneItemData"))
      (:Metal
        (LocalVariables|SetLocalItemToGive "/Game/Items/ResourcePickups/MetalItemData.MetalItemData"))
      (:Permanite
        (Development|PrintString "***WARNING*** - Invalid Resource type to Give to Player!")
        (return false))
      (:GoldCurrency
        (LocalVariables|SetLocalItemToGive "/Game/Items/ResourcePickups/GoldItemData.GoldItemData"))
      (:None
        _))
    (if (Utilities|IsValid _returnvalue :!class /Script/Engine.KismetSystemLibrary)
      (LocalVariables|SetInventoryOwner _returnvalue)
      (else
        (bind _asfort_athena_aibot_controller (Utilities|Casting|CastToFortAthenaAIBotController (Fortnite|GetControllerfromActor (Variables|Default|GetAbilityOwner)))
          (:then
            _)
          (:CastFailed
            _))
        (LocalVariables|SetInventoryOwner _asfort_athena_aibot_controller)))
    (Inventory|GiveItemtoInventoryOwner :InventoryOwner (LocalVariables|GetInventoryOwner) :ItemDefinition (LocalVariables|GetLocalItemToGive) :NumberToGive (Variables|GetGiveAmount) :bNotifyPlayer true :ItemLevel -1 :PickupInstigatorHandle -1 :bUseItemPickupAnalyticEvent false :WeaponAmmoOverride -1)
    (return true)
    (else
      (return false))))

(fn ShouldAbilityRespondToEvent
  (params (ActorInfo :type Gameplay%20Ability%20Actor%20Info%20Structure) (Payload :type Gameplay%20Event%20Data%20Structure))
  (bind (_owneractor _avataractor) (Utilities|Struct|BreakGameplayAbilityActorInfo ActorInfo))
  (bind _outxy (DataTable|EvaluateCurveTableRow "/Game/Athena/Balance/DataTables/AthenaGameData.AthenaGameData" "Default.Siphon.Enabled")
    (:RowFound
      (LocalVariables|SetbLocalHealthOnKill (== _outxy 1.0)))
    (:RowNotFound
      _))
  (bind _outxy_2 (DataTable|EvaluateCurveTableRow "/Game/Athena/Balance/DataTables/AthenaGameData.AthenaGameData" "Default.Siphon.GrantMatsOnKillEnabled")
    (:RowFound
      (LocalVariables|SetbLocalMatsOnKill (== _outxy_2 1.0)))
    (:RowNotFound
      _))
  (bind _asfort_player_pawn (Utilities|Casting|CastToFortPlayerPawn (Pawn|GetControlledPawn (Fortnite|GetControllerfromActor _avataractor)))
    (:then
      _)
    (:CastFailed
      _))
  (LocalVariables|SetbLocalShouldActivate (and (not (Health|IsDBNO _asfort_player_pawn)) (Math|Boolean|ORBoolean (LocalVariables|GetbLocalHealthOnKill) (Variables|MapState|GetbDebugBypasLocalMapEnable) (LocalVariables|GetbLocalMatsOnKill))))
  (return (LocalVariables|GetbLocalShouldActivate)))

(fn RestoreHealth
  (bind _outxy (DataTable|EvaluateCurveTableRow "/Game/Athena/Balance/DataTables/AthenaGameData.AthenaGameData" "Default.Siphon.InitialHealAmount")
    (:RowFound
      _)
    (:RowNotFound
      (Development|PrintString "Default Heal Amount Not Found")))
  (LocalVariables|SetfInitialAmountToHeal (select (Variables|MapState|GetIsCreativeOrPlayground) (Class|FortGameStateAthena|GetOverridenSiphonHealthValue (Variables|MapState|GetGameState) (Variables|Default|GetAbilityOwner)) _outxy))
  (if (> (LocalVariables|GetfInitialAmountToHeal) 0.0)
    (bind (_maxhealth _currenthealth _healthdelta _excesshealh) (CallFunction|DetermineHealthDelta (LocalVariables|GetfInitialAmountToHeal)))
    (LocalVariables|SetfShieldRestored _excesshealh)
    (bind _output_get (LocalVariables|SetfActualAmountHealed (Math|Float|Min(Float) _healthdelta (LocalVariables|GetfInitialAmountToHeal))))
    (if (> _output_get 0.0)
      (bind _returnvalue (Ability|GameplayEffect|AssignTagSetbyCallerMagnitude (Ability|MakeOutgoingGameplayEffectSpec (Variables|HealPlayer|GetGE_HealPlayer)) "(TagName=\"SetByCaller.Healing\")" (LocalVariables|GetfActualAmountHealed)))
      (Ability|ApplyGameplayEffectSpecToOwner _returnvalue)
      (Development|PrintString (Utilities|String|Append "Heal Player for " (Utilities|String|ToString(Float) (LocalVariables|GetfActualAmountHealed)) ". I expect that " (Utilities|String|ToString(Float) (LocalVariables|GetfShieldRestored)) " more health will transfer to shields."))
      (return true (LocalVariables|GetfInitialAmountToHeal) (LocalVariables|GetfActualAmountHealed) (LocalVariables|GetfShieldRestored))
      (else
        (Development|PrintString "Player is Full Life - All Healing Goes to Shields")
        (return true (LocalVariables|GetfInitialAmountToHeal) (LocalVariables|GetfActualAmountHealed) (LocalVariables|GetfShieldRestored))))
    (else
      (Development|PrintString "Heal Amount is Zero - Aborting Heal")
      (return false (LocalVariables|GetfInitialAmountToHeal) 0.0 0.0))))

(fn SetupRestoreModes
  (bind _asfort_game_state_athena (Utilities|Casting|CastToFortGameStateAthena (Game|GetGameState))
    (:then
      (Variables|MapState|SetGameState _asfort_game_state_athena)
      (Variables|MapState|SetIsCreativeOrPlayground (Fortnite|IsSpatialGameplayExperience))
      (bind _outxy (DataTable|EvaluateCurveTableRow "/Game/Athena/Balance/DataTables/AthenaGameData.AthenaGameData" "Default.Siphon.Enabled")
        (:RowFound
          (Variables|MapState|SetbAttemptHealthRestore (or (== _outxy 1.0) (Variables|MapState|GetbDebugBypasLocalMapEnable))))
        (:RowNotFound
          (Variables|MapState|SetbAttemptHealthRestore false)))
      (bind _outxy_2 (DataTable|EvaluateCurveTableRow "/Game/Athena/Balance/DataTables/AthenaGameData.AthenaGameData" "Default.Siphon.GrantMatsOnKillEnabled")
        (:RowFound
          (Variables|MapState|SetbAttemptMatsRestore (or (== _outxy_2 1.0) (Variables|MapState|GetbDebugBypasLocalMapEnable))))
        (:RowNotFound
          (Variables|MapState|SetbAttemptMatsRestore false)))
      (return (or (Variables|MapState|GetbAttemptHealthRestore) (Variables|MapState|GetbAttemptMatsRestore))))
    (:CastFailed
      (Development|PrintString "Not an Athena Game State - Aborting.")
      (return false))))
